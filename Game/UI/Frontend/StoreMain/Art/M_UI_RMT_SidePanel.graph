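# Translucent UI material for a side panel in the frontend store screen, with a two-color gradient and angled edges.
# A static switch picks an aspect-corrected radial gradient or a horizontal power-curve gradient as the blend between RGB1 and RGB2.
# Rotated, offset 9-slice UVs form two sharpened edge masks sharing one sharpness value.
# The separator mask limits the color blend, while the right-edge mask is computed separately from the color.
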Material /Game/UI/Frontend/StoreMain/Art/M_UI_RMT_SidePanel
graph {
 properties {
  bUsedWithStaticMesh=false
  blendMode=BLEND_Translucent
  materialDomain=MD_UI
  pixelDepthOffsetMode=PDOM_Legacy
  usedShadingModels=Default Lit
 }
 nodes [
  n1 = MaterialFunctionCall(Float2=MaterialFunctionCall() {materialFunction=&GetUserInterfaceUV}.Pixel Size) {materialFunction=&BreakOutFloat2Components}
  n2 = Max(A=@n1.G)
  n3 = MaterialFunctionCall(Float2=TextureCoordinate() {coordinateIndex=4}) {materialFunction=&BreakOutFloat2Components}
  n4 = AppendVector(A=Multiply(A=Divide(A=@n1.R, B=@n2), B=@n3.R), B=@n3.G)
  n5 = AppendVector(
   A=ScalarParameter() {defaultValue=0.5, group=Radial Gradient, parameterName=RadialGradient_u}
   B=ScalarParameter() {defaultValue=0.5, group=Radial Gradient, parameterName=RadialGradient_V}
  )
  n6 = ScalarParameter() {defaultValue=0.75, group=Radial Gradient, parameterName=Radial Gradient Radius_V}
  n7 = ScalarParameter() {defaultValue=1, group=Radial Gradient, parameterName=Radial Gradient Density}
  n8 = OneMinus(None=Saturate(None=MaterialFunctionCall(UVs=@n4, CenterPosition=@n5, Radius=@n6, Density=@n7) {materialFunction=&RadialGradientExponential}.RadialGradientExponential))
  n9 = Divide(A=ScalarParameter() {group=Degrees, parameterName=SeparatorAngle}) {constB=360}
  n10 = MaterialFunctionCall() {materialFunction=&GetUserInterfaceUV}
  n11 = ComponentMask(None=Add(
    A=Rotator(Coordinate=@n10.9-Slice UV, Time=@n9) {centerX=0, speed=6.2832}
    B=ScalarParameter() {defaultValue=0.1, group=Offset, parameterName=SeparatorOffset}
   )) {r=true}
  n12 = ScalarParameter() {defaultValue=32, group=Feather, parameterName=VerticalSharpness}
  n13 = Saturate(None=Multiply(A=OneMinus(None=@n11), B=@n12))
  n14 = ComponentMask(None=@n10.Normalized UV) {r=true}
  n15 = ScalarParameter() {defaultValue=2, parameterName=GradientPower}
  n16 = Saturate(None=LinearInterpolate(B=ScalarParameter() {parameterName=GradientOpacity}, Alpha=Power(Base=@n14, Exp=@n15)) {constA=1, constB=0.25})
  LinearInterpolate(
   A=VectorParameter() {parameterName=RGB2, defaultValue=<rgba 1 0 0.5571 1>}.RGB
   B=VectorParameter() {parameterName=RGB1, defaultValue=<rgba 0.0075 0.0176 0.2086 1>}.RGB
   Alpha=StaticSwitchParameter(True=Multiply(A=@n8, B=@n13), False=Multiply(A=@n16, B=@n13)) {parameterName=Is Circular Gradient}
  )
  n17 = Divide(A=ScalarParameter() {group=Degrees, parameterName=RightAngle}) {constB=360}
  n18 = ComponentMask(None=Add(
    A=Rotator(Coordinate=@n10.9-Slice UV, Time=@n17) {centerX=0, speed=6.2832}
    B=ScalarParameter() {defaultValue=0.1, group=Offset, parameterName=RightOffset}
   )) {r=true}
  Saturate(None=Multiply(A=OneMinus(None=@n18), B=@n12))
 ]
}
